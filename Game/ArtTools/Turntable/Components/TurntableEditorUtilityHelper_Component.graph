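# Editor utility component for the turntable art tool that gathers Content Browser selections for its owning actor.
# It appends selected assets to the owning MultipassTurntableActor_MultiCapture's MultiCaptureActorClasses as spawn structs with default transforms.
# Its helper receives static and skeletal mesh filters, but its branch returns the raw selection whenever filters are supplied.
# The selectActorsInLevel and syncAssetsInContentBrowser functions are stubs returning false.
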EditorUtilityBlueprint /Game/ArtTools/Turntable/Components/TurntableEditorUtilityHelper_Component
graphs [
 EventGraph {nodes [
   Event Begin Play()
   Event Tick!() -> (OutputDelegate, then, DeltaSeconds:float)
  ]}
 fillMultiCaptureActorsOnOwner {nodes [
   fillMultiCaptureActorsOnOwner() {is=title K2Node_FunctionEntry}
   > n1 = Cast To MultipassTurntableActor_MultiCapture!(execute, Object=GetOwner()) -> (then, CastFailed, AsMultipass Turntable Actor Multi Capture:Multipass Turntable Actor Multi Capture Object Reference) {is=title K2Node_DynamicCast}
   > n2 = getSelectedAssetsInContentBrowserWithFilter!(
    execute
    self
    filterClasses:Array of Object Class References=Make Array!("[0]":Object Class Reference=&StaticMesh, "[1]":Object Class Reference=&SkeletalMesh) -> (Array:Array of Object Class References)
   ) -> (then, assets:Array of Object References)
   > n3 = For Each Loop(Array:Array of Object References=@n2) -> (Array Element:Object Reference)
   > Add!(
    execute
    TargetArray:Array of Multipass TT Spawn Struct Structures=Get MultiCaptureActorClasses!(self:Multipass Turntable Actor Multi Capture Object Reference=@n1) -> (MultiCaptureActorClasses:Array of Multipass TT Spawn Struct Structures)
    NewItem_OptionalName_2_#1:Name=None
    NewItem_asset_18_#2:Object Reference=@n3
    NewItem_transform_14_#3:Transform="0,0,0|0,0,0|1,1,1"
    NewItem_animation_17_#4:Anim Sequence Base Object Reference
   ) -> (then, ReturnValue) {is=title K2Node_CallArrayFunction}
  ]}
 getSelectedAssets {nodes [
   getSelectedAssets() {is=title K2Node_FunctionEntry}
   > n1 = getSelectedAssetsInContentBrowserWithFilter(filterClasses=Make Array("[0]"=&StaticMesh, "[1]"=&SkeletalMesh))
   > Return Node!(execute, assets:Array of Object References=@n1)
  ]}
 getSelectedAssetsInContentBrowser {nodes [
   getSelectedAssetsInContentBrowser() {is=title K2Node_FunctionEntry}
   > n1 = GetSelectedAssets()
   > Return Node(assets=@n1)
  ]}
 getSelectedAssetsInContentBrowserWithFilter {nodes [
   getSelectedAssetsInContentBrowserWithFilter!() -> (then, filterClasses:Array of Object Class References) {is=title K2Node_FunctionEntry}
   n1 = Get filterClasses() -> (filterClasses:Array of Object Class References)
   > n2 = Branch(Condition="integer > integer"(A=Length(TargetArray:Array of Object Class References=@n1), B=0) {is=title K2Node_CallFunction})
   > n3 = GetSelectedAssets()
   > Return Node(assets=@n3)
   n4 = GetSelectedAssets(execute=@n2.else)
   > Set l_selectedAssets(l_selectedAssets:Array of Object References=@n4) -> (Output_Get:Array of Object References)
   > n5 = For Each Loop(Array:Array of Object Class References=Get filterClasses()) -> (Array Element:Object Class Reference)
   Return Node(execute=@n5.Completed, assets=Get l_finalAssets() -> (l_finalAssets:Array of Object References))
   > n6 = Filter by Class!(
    execute
    TargetArray:Array of Object References=Get l_selectedAssets() -> (l_selectedAssets:Array of Object References)
    ObjectClass:Object Class Reference=@n5
    FilterType:EEditorScriptingFilterType Enum
   ) -> (then, ReturnValue:Array of Object References)
   > Append Array!(execute, TargetArray:Array of Object References=Get l_finalAssets(), SourceArray:Array of Object References=@n6) -> (then)
  ]}
 selectActorsInLevel {nodes [
   selectActorsInLevel!() -> (then, actors:Array of Actor Object References) {is=title K2Node_FunctionEntry}
   > Return Node!(execute, " ":bool=false)
  ]}
 syncAssetsInContentBrowser {nodes [
   syncAssetsInContentBrowser!() -> (then, assets:Array of Strings) {is=title K2Node_FunctionEntry}
   > Return Node(" "=false)
  ]}
]
variables []
